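# Translucent, lit water material for the RustedTrack water slide, with index-of-refraction distortion.
# A tiled, panning flow-warp noise, scaled by a strength vector and vertex normals, forms the world position offset ripple.
# Fresnel tints the water color and, with depth fade, drives opacity, masked by UV gradients that soften all four edges.
# A second depth fade blends refraction near intersecting geometry, while parameters expose specular, roughness and anisotropy.
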Material /Hera/Sets/RustedTrack/Materials/M_RustedTrack_WaterSlide_Water
graph {
 properties {
  bAutomaticallySetUsageInEditor=false
  blendMode=BLEND_Translucent
  refractionMethod=RM_IndexOfRefraction
  translucencyLightingMode=TLM_SurfacePerPixelLighting
  usedShadingModels=Default Lit
 }
 nodes [
  LinearInterpolate(Alpha=DepthFade(FadeDistance=ScalarParameter() {defaultValue=25, parameterName=FadeDistance})) {constA=1, constB=1.1}
  n1 = ScalarParameter() {defaultValue=1, parameterName=U_Tile}
  n2 = ScalarParameter() {defaultValue=1, parameterName=V_Tile}
  n3 = Panner(Coordinate=Multiply(A=TextureCoordinate() {vTiling=4}, B=AppendVector(A=@n1, B=@n2))) {speedY=1}
  Multiply(
   A=Multiply(
    A=TextureSample(UVs=@n3) {samplerType=SAMPLERTYPE_Normal, texture=/Game/Global/Textures/Default/Noise/T_FlowWarp_Noise}.RGB
    B=VectorParameter() {parameterName=WPO_Strength, defaultValue=<rgba 5 25 -25 1>}.RGB
   )
   B=VertexNormalWS()
  )
  n4 = Fresnel() {exponent=1}
  Multiply(A=@n4, B=VectorParameter() {parameterName=WaterColor, defaultValue=<rgba 1 1 1 1>}.RGB)
  n5 = ScalarParameter() {defaultValue=5, parameterName=FinalOpacity}
  n6 = ScalarParameter() {defaultValue=25, parameterName=FadeDistance}
  n7 = MaterialFunctionCall(UV Channel=TextureCoordinate()) {materialFunction=&LinearGradient}
  n8 = Multiply(A=Multiply(A=@n7.VGradient, B=OneMinus(None=@n7.VGradient)), B=Multiply(A=@n7.UGradient, B=OneMinus(None=@n7.UGradient)))
  n9 = ScalarParameter() {defaultValue=1, parameterName=MaskPower}
  Multiply(A=Multiply(A=Multiply(A=@n4, B=@n5), B=DepthFade(FadeDistance=@n6)), B=Saturate(None=Power(Base=@n8, Exp=@n9)))
  ScalarParameter() {defaultValue=1, parameterName=Specular}
  ScalarParameter() {parameterName=Roughness}
  ScalarParameter() {defaultValue=1, parameterName=Anisotropy}
 ]
}
